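# Translucent unlit Niagara sprite material, an example of sprites orbiting around their particle position in the shader.
# RotateAboutAxis, minus the radius offset, yields the orbit displacement from time, Rotate Radius and Rotate Speed dynamic parameters, and a local-space Axis parameter bindable through the renderer.
# Particle color tints the editor spawn-point texture through an emissive adjustment function.
# Opacity combines particle and texture alpha, squared near-camera fade and depth fade.
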Material /Game/Effects/Niagara/Systems/Examples/M_Sample_RotateAroundPoint
graph {
 properties {
  bAutomaticallySetUsageInEditor=false
  bUsedWithNiagaraSprites=true
  blendMode=BLEND_Translucent
  shadingModel=MSM_Unlit
  usedShadingModels=Unlit
 }
 nodes [
  n1 = ParticleColor()
  n2 = TextureSample() {texture=/Engine/EditorResources/Ai_Spawnpoint}
  n3 = MaterialFunctionCall(Distance=ScalarParameter() {defaultValue=150, parameterName=Near Camera Fade}) {materialFunction=&Near_Camera_Fade}
  DepthFade(
   Opacity=Multiply(A=Multiply(A=@n1.A, B=@n2.A), B=Multiply(A=@n3.Result, B=@n3.Result))
   FadeDistance=ScalarParameter() {defaultValue=50, parameterName=Depth Fade}
  )
  MaterialFunctionCall(Emissive=Multiply(A=@n1.RGB, B=@n2.RGB)) {materialFunction=&MF_FX_EmissiveAdjustments}
  n4 = VectorParameter() {
   desc="Specifying the rotation axis. By binding it via the Renderer, you can control it directly within Niagara."
   parameterName=Axis
   defaultValue=<rgba 0 0 1 1>
  }
  n5 = Transform(None=@n4.RGB) {transformSourceType=TRANSFORMSOURCE_Local}
  n6 = DynamicParameter() {defaultValue=<rgba 100 0.5 1 1>, paramNames [Rotate Radius, Rotate Speed, Param3, Param4]}
  n7 = Time()
  n8 = ParticlePositionWS()
  n9 = Multiply(
   A=DynamicParameter() {defaultValue=<rgba 100 0.5 1 1>, paramNames [Rotate Radius, Rotate Speed, Param3, Param4]}.Rotate Radius
   B=Transform(None=Constant3Vector() {constant=<rgba 0 1 0 0>}) {transformSourceType=TRANSFORMSOURCE_Local}
  )
  Subtract(
   A=RotateAboutAxis(
    NormalizedRotationAxis=Normalize(VectorInput=@n5)
    RotationAngle=Multiply(A=@n6.Rotate Speed, B=@n7)
    PivotPoint=Add(A=@n8, B=@n9)
    Position=@n8
   )
   B=@n9
  )
 ]
}
