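# Example Niagara system, named for a fake radial shadow technique, that bursts mesh cubes outward from a sphere.
# Drag, vortex force, random spin and curves for scale and color make the cubes swirl, cool from bright HDR to dark, and shrink away.
# Dynamic material parameter index 3 passes owner position and an animated random magnitude to materials, presumably driving the fake shadow.
# Cube uses two mesh renderers; Cube_LWE lists none.
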NiagaraSystem /Game/Effects/Niagara/Systems/Examples/NS_Fake_Radial_Shadow_Sample
dependencies {
 usedDataInterfaces [&NiagaraDataInterfaceCurve, &NiagaraDataInterfaceColorCurve]
 usedDynamicInputs [&MakeFloatFromVector, &ConvertPositionToVector]
 usedModules [
  &self:SystemSpawnScript.SystemScriptSource.SystemScriptGraph.NiagaraNodeAssignment_0.SetVariables_#1
  &SystemState
 ]
}
emitters {
 Cube {
  emitterUpdateScript [
   &EmitterState(Life Cycle Mode=::NewEnumerator1, Loop Behavior=::NewEnumerator1, Loop Duration=0.2, UseLoopDelay=false)
   &SpawnBurst_Instantaneous(Spawn Count=24, Use Spawn Probability=false, UseLoopCountLimit=false)
   /SetVariables(Emitter.Use Particle Color:NiagaraFloat=&FloatFromCurve(
     FloatCurve=DataInterface() {propertyValues {Curve {
        defaultValue=1
        keys [{arriveTangent=2.5795, interpMode=RCIM_Cubic, leaveTangent=2.5795, tangentMode=RCTM_User, time=0, value=0}, <at 0.5 1 RCIM_Cubic>]
       }}}
     CurveIndex=$Emitter.Age
    )) {isSetParameters=true}
  ]
  particleSpawnScript [
   &InitializeParticle(
    Color=<rgba 0.1557 0 1 1>
    Lifetime Max=1
    Lifetime Min=0.4
    Lifetime Mode=::NewEnumerator1
    Mesh Scale Mode=::NewEnumerator3
    Mesh Uniform Scale=10
    UsePositionOffset=false
    Write Material Random=true
   )
   &ShapeLocation(MaskSpawnGroup=false, Override Seed=false, Sphere Radius=50, UseOwnerScale=true)
   &AddVelocity(
    Module.Use Falloff Distance=false
    Override Seed=false
    Velocity Mode=::NewEnumerator1
    Velocity Speed=&RandomRangeFloat(Minimum=2000, Maximum=3000)
   )
   &InitialMeshOrientation(Use Rotation Vector=false)
  ]
  particleUpdateScript [
   &ParticleState()
   &ScaleMeshSize(Scale Factor=&Multiply_VectorByFloat(
     Vector=<1 1 1>
     Float=&FloatFromCurve(FloatCurve=DataInterface() {propertyValues {Curve {keys [
          {tangentMode=RCTM_User, time=0, value=1}
          {arriveTangent=-0.0405, interpMode=RCIM_Cubic, leaveTangent=-0.0405, tangentMode=RCTM_User, time=0.7009, value=1}
          {arriveTangent=-8.0089, interpMode=RCIM_Cubic, leaveTangent=-8.0089, tangentMode=RCTM_User, time=1, value=0}
         ]}}})
    ))
   &UpdateMeshOrientation(
    Rotation Rate=&Multiply_Float(
     A=&RandomRangeFloat(Minimum=-2, Maximum=2)
     B=&FloatFromCurve(FloatCurve=DataInterface() {propertyValues {Curve {keys [
          {tangentMode=RCTM_User, time=0, value=1}
          {arriveTangent=-0.0405, interpMode=RCIM_Cubic, leaveTangent=-0.0405, tangentMode=RCTM_User, time=0.5, value=1}
          {arriveTangent=-2.2223, interpMode=RCIM_Cubic, leaveTangent=-2.2223, tangentMode=RCTM_User, time=0.9, value=0}
         ]}}})
    )
    Rotation Vector=&RandomRangeVector(Minimum=<0 0 0>, Maximum=<1 1 1>)
   )
   &Drag(Drag=15, Use Linear Drag=true, Use Rotational Drag=true)
   &VortexForce(Origin Pull Amount=5, UseOriginOffset=false, Vortex Force Amount=8000)
   &SolveForcesAndVelocity(Clamp Velocity=false, Limit Acceleration=false)
   &DynamicMaterialParameters(
    Index 3 Param 4=&Multiply_Float(
     A=&RandomRangeFloat(Minimum=70, Maximum=200)
     B=&FloatFromCurve(FloatCurve=DataInterface() {propertyValues {Curve {keys [
          <at 0 1 RCIM_Cubic>
          {arriveTangent=-1.1111, interpMode=RCIM_Cubic, leaveTangent=-1.1111, time=0.6, value=2.375}
          {arriveTangent=-8.0089, interpMode=RCIM_Cubic, leaveTangent=-8.0089, tangentMode=RCTM_User, time=0.9, value=0}
         ]}}})
    )
    Index 3 Vector 1:NiagaraPosition=$Engine.Owner.Position
    Module.Index3Float3Write=true
    Module.Index3VectorWrite=true
    Write Parameter Index 0=::NewEnumerator0
    Write Parameter Index 3=::NewEnumerator3
   )
   &Color(Color=&ColorFromCurve(ColorCurve=DataInterface() {propertyValues {
       AlphaCurve {keys [<at 0.8016 1>,<at 1 0>]}
       BlueCurve {keys [<at 0.0128 8.2795>, {leaveTangent=-1373.7484, time=0.1938, value=0}, <at 0.5203 0.0191>]}
       GreenCurve {keys [<at 0.0128 9.8412>, {leaveTangent=-1373.7484, time=0.1938, value=1.9299}, <at 0.5203 0.0059>]}
       RedCurve {keys [<at 0.0128 10>, {leaveTangent=-1373.7484, time=0.1938, value=5}, <at 0.5203 0.0143>]}
      }}))
  ]
  rendererClasses [&NiagaraMeshRendererProperties]
  renderers [
   {rendererClass=&NiagaraMeshRendererProperties, rendererIndex=0}
   {rendererClass=&NiagaraMeshRendererProperties, rendererIndex=1}
  ]
 }
 Cube_LWE {rendererClasses [], renderers []}
}
summary {
 systemName=NS_Fake_Radial_Shadow_Sample
 emitters [{emitterName=Cube_LWE, rendererClasses []}, {emitterName=Cube, rendererClasses [&NiagaraMeshRendererProperties]}]
 userVariables []
}
