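# Niagara example system that, by its name, shows samples of faking rotation around a point.
# It holds three emitters named for alternative techniques: shape location, mesh offset and rotation, and World Position Offset rotation.
# A Set Variables assignment in system spawn accompanies System State, with Make Quaternion and Multiply Float as the dynamic inputs in use.
# No user parameters are exposed, and emitter internals and renderers are not shown.
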NiagaraSystem /Game/Effects/Niagara/Systems/Examples/NS_Fake_RotateAroundPoint_Samples
dependencies {
 usedDataInterfaces []
 usedDynamicInputs [&MakeQuaternion, &Multiply_Float]
 usedModules [
  &SystemState
  &self:SystemSpawnScript.SystemScriptSource.SystemScriptGraph.NiagaraNodeAssignment_0.SetVariables_#1
 ]
}
emitters {
 By_WPO_Rotation {rendererClasses [], renderers []}
 Mesh_Offset_and_Rotation {rendererClasses [], renderers []}
 Using_ShapeLocation {rendererClasses [], renderers []}
}
summary {
 systemName=NS_Fake_RotateAroundPoint_Samples
 emitters [
  {emitterName=Using_ShapeLocation, rendererClasses []}
  {emitterName=Mesh_Offset_and_Rotation, rendererClasses []}
  {emitterName=By_WPO_Rotation, rendererClasses []}
 ]
 userVariables []
}
