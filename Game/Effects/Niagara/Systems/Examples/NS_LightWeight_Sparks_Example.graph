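# Niagara example system for lightweight spark effects, with four emitters named for radial, simple impact and two directional mesh spark variants.
# It exposes user parameters for global scale and sparks direction.
# The system level uses SystemState and a spawn-script variable assignment, with VectorFromFloat and a scratch dynamic input available to compute values.
# The emitters list no renderers or modules here, so their internal behaviour is not visible.
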NiagaraSystem /Game/Effects/Niagara/Systems/Examples/NS_LightWeight_Sparks_Example
dependencies {
 usedDataInterfaces []
 usedDynamicInputs [
  &VectorFromFloat
  &self:ScratchDynamicInput
 ]
 usedModules [
  &self:SystemSpawnScript.SystemScriptSource.SystemScriptGraph.NiagaraNodeAssignment_0.SetVariables_#1
  &SystemState
 ]
}
emitters {
 Directional_Mesh_Sparks1 {rendererClasses [], renderers []}
 Directional_Mesh_Sparks2 {rendererClasses [], renderers []}
 Radial_Mesh_Sparks {rendererClasses [], renderers []}
 Simple_Impact_Mesh_Sparks0 {rendererClasses [], renderers []}
}
summary {
 systemName=NS_LightWeight_Sparks_Example
 emitters [
  {emitterName=Radial_Mesh_Sparks, rendererClasses []}
  {emitterName=Simple_Impact_Mesh_Sparks0, rendererClasses []}
  {emitterName=Directional_Mesh_Sparks1, rendererClasses []}
  {emitterName=Directional_Mesh_Sparks2, rendererClasses []}
 ]
 userVariables [
  User.Global Scale {type=NiagaraFloat, defaultValue=1, description=""}
  User.Sparks Direction {type=Vector3f, description="", defaultValue {struct=Vector3f, value=<1 0 0>}}
 ]
}
