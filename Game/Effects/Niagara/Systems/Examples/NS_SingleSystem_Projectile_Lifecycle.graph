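# Example Niagara system that plays a projectile's whole visual lifecycle in one system, with a User Finish Trigger input.
# System-level 3Steps modules accompany a shared Timeline Age that indexes size, color and alpha curves across a 0–3 range.
# The burst-spawned Core sprite ignores lifetime expiry, persisting until a system kill flag removes it.
# GPU Small_Particles spawn in a sphere at a timeline-driven rate, drifting with curl noise while fading.
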NiagaraSystem /Game/Effects/Niagara/Systems/Examples/NS_SingleSystem_Projectile_Lifecycle
dependencies {
 usedDataInterfaces [&NiagaraDataInterfaceCurve, &NiagaraDataInterfaceColorCurve]
 usedDynamicInputs [&Vector4FromCurve]
 usedModules [
  &self:Initialize_3Steps_Set
  &self:SystemSpawnScript.SystemScriptSource.SystemScriptGraph.NiagaraNodeAssignment_3.SetVariables_#1
  &SystemState
  &self:Update_3Steps_Set
 ]
}
emitters {
 Core {
  emitterUpdateScript [
   &EmitterState(Life Cycle Mode=::NewEnumerator1, Loop Behavior=::NewEnumerator1, Loop Duration=0.1, UseLoopDelay=false)
   &SpawnBurst_Instantaneous(Use Spawn Probability=false, UseLoopCountLimit=false)
  ]
  particleSpawnScript [&InitializeParticle(
    Color=<rgba 0 1 0.8978 0.5>
    Lifetime=$Emitter.CurrentLoopDuration
    Sprite Size Mode=::NewEnumerator3
    Uniform Sprite Size=100
    UsePositionOffset=false
    Write Material Random=true
   )]
  particleUpdateScript [
   &KillParticles(Kill Particles=$System.IS_Kill Particles)
   &ParticleState(Kill Particles When Lifetime Has Elapsed=false)
   &SolveForcesAndVelocity(Clamp Velocity=false, Limit Acceleration=false)
   &ScaleSpriteSize(
    Scale Sprite Size Mode=::NewEnumerator0
    Uniform Scale Factor=&FloatFromCurve(
     FloatCurve=DataInterface() {propertyValues {Curve {keys [<at 0 0.2>,<at 0.2 4>,<at 0.9 1>,<at 1 1>,<at 1.0795 1.25>,<at 1.1357 0.75>,<at 1.2085 1.15>,<at 1.272 0.75>,<at 1.3409 1.625>,<at 1.4352 0.75>,<at 1.491 1.025>,<at 1.5846 0.75>,<at 1.6529 1.625>,<at 1.7339 0.75>,<at 1.8267 1.175>,<at 1.9079 0.625>,<at 2 1>,<at 2.1866 2.625>,<at 2.4 1>,<at 3 0>]}}}
     CurveIndex=$System.Timeline Age
    )
   )
   &ScaleColor(
    Scale Alpha=&FloatFromCurve(
     FloatCurve=DataInterface() {propertyValues {Curve {keys [
         <at 0 0>
         <at 0.7 0.5>
         {arriveTangent=0.6667, interpMode=RCIM_Cubic, leaveTangent=0.6667, time=1, value=1}
         {arriveTangent=0.4167, interpMode=RCIM_Cubic, leaveTangent=0.4167, time=1.3, value=0.9}
         {arriveTangent=0.1429, interpMode=RCIM_Cubic, leaveTangent=0.1429, time=1.6, value=1.25}
         {arriveTangent=-1.4644, tangentMode=RCTM_User, time=2, value=1}
         <at 3 0>
        ]}}}
     CurveIndex=$System.Timeline Age
    )
    Scale RGB=&MakeVectorFromLinearColorRGB(Color=&ColorFromCurve(ColorCurve=$System.Curve for Colors, CurveIndex=$System.Timeline Age))
    ScaleA=true
    ScaleRGB=true
   )
   /SetVariables(
    Particles.Core Color:LinearColor=&ColorFromCurve(ColorCurve=$System.Curve for Colors, CurveIndex=$System.Timeline Age)
    Particles.Core Size:Vector2f=&Multiply_Vector2D_ByFloat(
     Float=&FloatFromCurve(FloatCurve=$System.3Steps Curve, CurveIndex=$System.Timeline Age)
     Vector2D=<40 40>
    )
   ) {isSetParameters=true}
  ]
  rendererClasses [&NiagaraSpriteRendererProperties]
  renderers [
   {rendererClass=&NiagaraSpriteRendererProperties}
   {rendererClass=&NiagaraSpriteRendererProperties, rendererIndex=1}
  ]
 }
 Small_Particles {
  simTarget=GPUComputeSim
  emitterUpdateScript [
   &EmitterState(Life Cycle Mode=::NewEnumerator1, Loop Duration=1, UseLoopDelay=false)
   &SpawnRate(
    SpawnRate=&FloatFromCurve(
     FloatCurve=DataInterface() {propertyValues {Curve {keys [<at 0 2>,<at 0.2 1>,<at 2 1>,<at 2.8 0>]}}}
     CurveIndex=$System.Timeline Age
     Scale Curve=200
    )
    Use Spawn Probability=false
   )
  ]
  particleSpawnScript [
   &InitializeParticle(
    Color=<rgba 0 0.1666 1 1>
    Lifetime Max=0.5
    Lifetime Min=0.2
    Lifetime Mode=::NewEnumerator1
    Mass Mode=::NewEnumerator3
    Sprite Rotation Mode=::NewEnumerator2
    Sprite Size Mode=::NewEnumerator1
    Uniform Sprite Size Max=4
    Uniform Sprite Size Min=2
    UsePositionOffset=false
    Write Material Random=true
   )
   &ShapeLocation(MaskSpawnGroup=false, Override Seed=false, Sphere Radius=40, UseOwnerScale=true)
  ]
  particleUpdateScript [
   &ParticleState()
   &ScaleColor(
    Scale Alpha=&FloatFromCurve(FloatCurve=DataInterface() {propertyValues {Curve {keys [<at -0.0046 1.0113 RCIM_Cubic>,<at 1.0011 0.002 RCIM_Cubic>]}}})
    ScaleA=true
    ScaleRGB=true
   )
   &CurlNoiseForce(Bias Noise=false, Noise Quality / Cost=::NewEnumerator0, Pan Noise=false)
   &SolveForcesAndVelocity(Clamp Velocity=false, Limit Acceleration=false)
  ]
  rendererClasses [&NiagaraSpriteRendererProperties]
  renderers [{rendererClass=&NiagaraSpriteRendererProperties}]
 }
}
summary {
 systemName=NS_SingleSystem_Projectile_Lifecycle
 emitters [
  {emitterName=Core, rendererClasses [&NiagaraSpriteRendererProperties]}
  {emitterName=Small_Particles, simTarget=GPUComputeSim, rendererClasses [&NiagaraSpriteRendererProperties]}
 ]
 userVariables [User.Finish Trigger {type=NiagaraBool}]
}
